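; Window wall piece for the GlowDonut environment set with two glass provider components, configured entirely in its construction script.
; AdjustGlassTransparencyOrder sets translucent sort priority on the static mesh and valid glass providers to 1 when the actor's forward vector leans toward world X, otherwise 0.
; The construction script then hands both glass providers to SimpleDestructiblePropSubsystem for conversion to destructible prop data, if the subsystem is valid.
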(event EventBeginPlay
  (|CallParentFunction) ; This node is disabled and will not be called.\nDrag off pins to build functionality.
)

(event Collision|EventActorBeginOverlap
  (params (OtherActor :type Actor%20Object%20Reference))
  (|CallParentFunction OtherActor) ; This node is disabled and will not be called.\nDrag off pins to build functionality.
)

(event EventTick
  (params (DeltaSeconds :type Float%20%28single-precision%29))
  (|CallParentFunction DeltaSeconds) ; This node is disabled and will not be called.\nDrag off pins to build functionality.
)

(fn AdjustGlassTransparencyOrder
  (bind _returnvalue (Transformation|GetActorRotation))
  (bind _returnvalue_2 (Math|Vector|GetForwardVector _returnvalue))
  (bind _returnvalue_3 (Math|Vector|Normalize _returnvalue_2))
  (bind _returnvalue_4 (Math|Conversions|ToVector2D(Vector) _returnvalue_3))
  (bind _returnvalue_5 (Math|Vector2D|GetAbs2D _returnvalue_4))
  (bind (_x _y) (Math|Vector2D|BreakVector2D _returnvalue_5))
  (bind _returnvalue_6 (>= _x _y))
  (bind _returnvalue_7 (Math|Conversions|ToInteger(Boolean) _returnvalue_6))
  (bind _glassproviderb (Variables|Default|GetGlassProviderB))
  (bind _glassprovidera (Variables|Default|GetGlassProviderA))
  (Rendering|SetTranslucentSortPriority (Variables|Debug|GetStaticMeshComponent) _returnvalue_7)
  (Utilities|IsValid _glassproviderb
    (:Is%20Valid
      (Rendering|SetTranslucentSortPriority _glassproviderb _returnvalue_7))
    (:Is%20Not%20Valid
      _))
  (Utilities|IsValid _glassprovidera
    (:Is%20Valid
      (Rendering|SetTranslucentSortPriority _glassprovidera _returnvalue_7))))

(fn ConstructionScript
  (|CallParentFunction)
  (CallFunction|AdjustGlassTransparencyOrder)
  (bind _returnvalue (Class|SimpleDestructiblePropSubsystem|GetSimpleDestructiblePropSubsystem))
  (Utilities|IsValid _returnvalue
    (:Is%20Valid
      (Class|SimpleDestructiblePropSubsystem|ConstructionScriptConverttoSimpleDestructiblePropData _returnvalue self (Utilities|Array|MakeArray (Variables|Default|GetGlassProviderB) (Variables|Default|GetGlassProviderA))))))
