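; Window wall piece for the GlowDonut environment set that registers its glass as a simple destructible prop.
; The construction script calls the parent, then gets the SimpleDestructiblePropSubsystem and, if valid, passes itself and an array holding GlassProviderB to convert to destructible prop data.
; BeginPlay, overlap and Tick are disabled stubs, so all logic runs at construction.
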(event EventBeginPlay
  (|CallParentFunction) ; This node is disabled and will not be called.\nDrag off pins to build functionality.
)

(event Collision|EventActorBeginOverlap
  (params (OtherActor :type Actor%20Object%20Reference))
  (|CallParentFunction OtherActor) ; This node is disabled and will not be called.\nDrag off pins to build functionality.
)

(event EventTick
  (params (DeltaSeconds :type Float%20%28single-precision%29))
  (|CallParentFunction DeltaSeconds) ; This node is disabled and will not be called.\nDrag off pins to build functionality.
)

(fn ConstructionScript
  (|CallParentFunction)
  (bind _returnvalue (Class|SimpleDestructiblePropSubsystem|GetSimpleDestructiblePropSubsystem))
  (Utilities|IsValid _returnvalue
    (:Is%20Valid
      (Class|SimpleDestructiblePropSubsystem|ConstructionScriptConverttoSimpleDestructiblePropData _returnvalue self (Utilities|Array|MakeArray (Variables|Default|GetGlassProviderB) 0)))))
